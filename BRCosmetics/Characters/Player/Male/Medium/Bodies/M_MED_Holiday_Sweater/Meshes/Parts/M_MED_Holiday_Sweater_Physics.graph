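# Physics asset for the male medium Holiday Sweater cosmetic, defining bodies and joints for dynamic earflap and top-flap bones around the head.
# Each side has a three-box earflap chain and a single-box top flap, each jointed to its own small base sphere.
# Tight angular limits restrict bending and twist, with secondary swing widening toward the earflap tips.
# A head capsule has no constraints, apparently serving as a collider for the flaps.
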PhysicsAsset /BRCosmetics/Characters/Player/Male/Medium/Bodies/M_MED_Holiday_Sweater/Meshes/Parts/M_MED_Holiday_Sweater_Physics
bodies {
 dyn_large_earflap_01_l [{shapeType=Box, extentX=3.9364, extentY=8.361, extentZ=3.6301, center=<1.8242 0 0.0001>}]
 dyn_large_earflap_01_r [{shapeType=Box, extentX=2.6651, extentY=8.6655, extentZ=2.2229, center=<1.2868 0 0>}]
 dyn_large_earflap_02_l [{shapeType=Box, extentX=3.5353, extentY=6.8156, extentZ=3.016, center=<1.497 0 0.0001>}]
 dyn_large_earflap_02_r [{shapeType=Box, extentX=4.1616, extentY=7.9459, extentZ=2.6368, center=<1.8689 0 0>}]
 dyn_large_earflap_base_l [{shapeType=Sphere, radius=0.5}]
 dyn_large_earflap_base_r [{shapeType=Sphere, radius=0.5}]
 dyn_large_earflap_parent_l [{shapeType=Box, extentX=3.2243, extentY=9.5, extentZ=2.5622, center=<1.7546 0 0>}]
 dyn_large_earflap_parent_r [{shapeType=Box, extentX=3.6651, extentY=8.6655, extentZ=2.2229, center=<1.869 0 0>}]
 dyn_top_flap_base_l [{shapeType=Sphere, radius=0.5}]
 dyn_top_flap_base_r [{shapeType=Sphere, radius=0.5}]
 dyn_top_flap_parent_l [{shapeType=Box, extentX=5.3359, extentY=7.9141, extentZ=2, center=<2.5089 -0.0001 0>}]
 dyn_top_flap_parent_r [{shapeType=Box, extentX=5.4076, extentY=5.9634, extentZ=2.0441, center=<2.6763 0.0001 0>}]
 head [{shapeType=Capsule, length=7.661, radius=7.2205, center=<0 3.5797 0.0001>, rotation=<rot 90 32.0054 53.0917>}]
}
constraints [
 {bone1Name=dyn_top_flap_parent_r, bone2Name=dyn_top_flap_base_r, swing1LimitDegrees=5, swing2LimitDegrees=20, twistLimitDegrees=2}
 {bone1Name=dyn_large_earflap_parent_l, bone2Name=dyn_large_earflap_base_l, swing1LimitDegrees=10, swing2LimitDegrees=10, twistLimitDegrees=3}
 {bone1Name=dyn_large_earflap_01_l, bone2Name=dyn_large_earflap_parent_l, swing1LimitDegrees=10, swing2LimitDegrees=15, twistLimitDegrees=3}
 {bone1Name=dyn_large_earflap_02_l, bone2Name=dyn_large_earflap_01_l, swing1LimitDegrees=10, swing2LimitDegrees=18, twistLimitDegrees=3}
 {bone1Name=dyn_large_earflap_parent_r, bone2Name=dyn_large_earflap_base_r, swing1LimitDegrees=10, swing2LimitDegrees=10, twistLimitDegrees=3}
 {bone1Name=dyn_large_earflap_02_r, bone2Name=dyn_large_earflap_01_r, swing1LimitDegrees=10, swing2LimitDegrees=18, twistLimitDegrees=3}
 {bone1Name=dyn_large_earflap_01_r, bone2Name=dyn_large_earflap_parent_r, swing1LimitDegrees=10, swing2LimitDegrees=15, twistLimitDegrees=3}
 {bone1Name=dyn_top_flap_parent_l, bone2Name=dyn_top_flap_base_l, swing1LimitDegrees=5, swing2LimitDegrees=20, twistLimitDegrees=2}
]
