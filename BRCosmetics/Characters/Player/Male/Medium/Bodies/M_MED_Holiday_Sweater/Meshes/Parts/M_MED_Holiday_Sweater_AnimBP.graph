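# Animation Blueprint for the male medium Holiday Sweater cosmetic body part, adding physics-driven secondary motion to the mesh.
# It copies the pose from the attached parent mesh, then runs a RigidBody node simulating the sweater physics asset in spine_05 bone space, with custom gravity and damped component motion.
# Jump begin and end events set Jump_Alpha after short retriggerable delays, switching an upward external force on and off in that simulation.
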AnimBlueprint /BRCosmetics/Characters/Player/Male/Medium/Bodies/M_MED_Holiday_Sweater/Meshes/Parts/M_MED_Holiday_Sweater_AnimBP
target_skeleton=&/M_MED_Holiday_Sweater_Skeleton
graphs [
 AnimGraph {nodes [
   n1 = RigidBody(
    ComponentPose=Local To Component(LocalPose=Copy Pose From Mesh() {bUseAttachedParent=true})
    ExternalForce=Get Jump_Alpha() -> (Jump_Alpha:Vector)
   ) {
    bClampLinearTranslationLimitToRefPose=true
    bOverrideWorldGravity=true
    overridePhysicsAsset=&/M_MED_Holiday_Sweater_Physics
    simulationSpace=BaseBoneSpace
    alphaBoolBlend {blendOption=Linear}
    baseBoneRef {boneName=spine_05}
    componentAppliedLinearAccClamp=<6000 6000 6000>
    componentLinearAccScale=<0.2 0.2 0.2>
    componentLinearVelScale=<0.2 0.2 0.2>
    overrideWorldGravity {z=-980}
   }
   Output Pose(Result=Component To Local(ComponentPose=@n1)) {layerGroup=None}
  ]}
 EventGraph {nodes [
   Event OnBeginJumping()
   > RetriggerableDelay(Duration=0.25)
   > Set Jump_Alpha(Jump_Alpha:Vector="0, 0,1250") -> (Output_Get:Vector)
   Event OnEndJumping()
   > RetriggerableDelay(Duration=0.375)
   > Set Jump_Alpha(Jump_Alpha="0, 0, 0")
   Event BlueprintUpdateAnimation()
  ]}
]
variables [Jump_Alpha {type=Vector}]
